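# Material function in the vehicle cosmetics parent materials that layers trim onto an incoming set of material attributes.
# It blends the input toward the output of MF_Trim_Base_v2 using the INTrimMask input.
# It then restores the input's original Normal and sets the clear coat bottom normal to a normalized angle-corrected combination of both layers, which the mask does not modulate.
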MaterialFunction /VehicleCosmetics/Materials/VehicleParents_Version2/Material_Functions/MF_Apply_Trim_v2
graph {nodes [
  n1 = GetMaterialAttributes(None=FunctionInput() {inputName=MaterialAttributes, inputType=FunctionInput_MaterialAttributes}) {attributeGetTypes [#1, #2]}
  n2 = GetMaterialAttributes(None=MaterialFunctionCall() {materialFunction=&/MF_Trim_Base_v2}.Result) {attributeGetTypes [#2]}
  n3 = FunctionInput() {inputName=INTrimMask, sortPriority=1}
  n4 = MaterialFunctionCall(BaseNormal=@n1.ClearCoatBottomNormal, AdditionalNormal=@n2.ClearCoatBottomNormal) {materialFunction=&BlendAngleCorrectedNormals}
  FunctionOutput(None=SetMaterialAttributes!(
    MaterialAttributes=BlendMaterialAttributes(A=@n1.MaterialAttributes, B=@n2.MaterialAttributes, Alpha=@n3)
    Normal=@n1.Normal
    ClearCoatBottomNormal=Normalize(VectorInput=@n4.Result)
   ) {attributeSetTypes [#1, #2]}) {outputName=SetMaterialAttributes}
 ]}
