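# Base material function for vehicle chassis shading, outputting material attributes plus trim, carbon fiber and plastics masks from a composite texture.
# A paint mask blends paint color over the base texture, then emissive parts, headlight, brake, reverse and boost glows combine masks, colors, brightness and Blueprint-only state parameters.
# Per-group switches let base color and emissive color override each other, with a legacy mode.
# Light masks optionally use UV cropping or quality-gated bump-offset parallax.
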MaterialFunction /VehicleCosmetics/Materials/VehicleParents_Version2/Material_Functions/MF_Chassis_Base_v2
graph {nodes [
  n1 = NamedRerouteUsage()
  Add(B=Multiply(A=NamedRerouteUsage(), B=@n1))
  n2 = VectorParameter() {group=Emissive, parameterName=EmissiveUVs_RG_UpperLeftCorner_BA_LowerRightCorner, defaultValue=<rgba 0 0 1 1>}
  n3 = AppendVector(A=@n2.RGB, B=@n2.RGB)
  n4 = TextureObjectParameter() {desc=_E, group=Textures, parameterName=EmissiveMasks, samplerType=SAMPLERTYPE_LinearColor, texture=/VehicleCosmetics/Textures/T_Black_Linear}
  Composite!(RGBA=@n3, T2D_In=@n4)
  Composite!()
  n5 = NamedRerouteUsage()
  n6 = NamedRerouteUsage()
  n7 = NamedRerouteUsage()
  n8 = Multiply(A=@n6, B=@n7)
  n9 = NamedRerouteUsage()
  n10 = Multiply(A=NamedRerouteUsage(), B=@n9)
  n11 = NamedRerouteUsage()
  n12 = NamedRerouteUsage()
  n13 = NamedRerouteUsage()
  n14 = Add(A=Add(A=Add(A=@n8, B=@n10), B=Multiply(A=@n11, B=@n12)), B=Multiply(A=@n6, B=@n13))
  MaterialExpressionReroute_2 = Reroute(None=@n14)
  n15 = Max(A=Max(A=@n7, B=@n9), B=@n12)
  n16 = Saturate(None=ConstantBiasScale(None=Max(A=@n15, B=@n13)) {bias=0, scale=2})
  MaterialExpressionReroute_3 = Reroute(None=@n16)
  n17 = LinearInterpolate(A=@n5, B=@MaterialExpressionReroute_2.EmissiveColors, Alpha=@MaterialExpressionReroute_3.EmissiveMasks)
  Composite!(BaseColorPainted=@n5, BaseColorEmissive=@n17)
  n18 = NamedRerouteUsage()
  ConstantBiasScale(None=@n18) {bias=0, scale=10}
  n19 = NamedRerouteUsage()
  ConstantBiasScale(None=@n19) {bias=0, scale=10}
  n20 = NamedRerouteUsage()
  ConstantBiasScale(None=@n20) {bias=0, scale=10}
  n21 = NamedRerouteUsage()
  ConstantBiasScale(None=@n21) {bias=0, scale=10}
  n22 = TextureSampleParameter2D() {
   group=Textures
   parameterName=ChassisComposite_Mask_LOD
   samplerSource=SSM_Wrap_WorldGroupSettings
   samplerType=SAMPLERTYPE_LinearColor
   texture=&T_Black_Linear
  }
  FunctionOutput(None=NamedRerouteDeclaration(None=@n22.G) {name=TrimMask, nodeColor=<rgba 0.026 0.026 0.026 1>}) {outputName=TrimMask}
  n23 = NamedRerouteUsage()
  n24 = SetMaterialAttributes!(
   MaterialAttributes
   BaseColor=NamedRerouteUsage()
   Metallic=NamedRerouteUsage()
   Roughness=NamedRerouteUsage()
   EmissiveColor=NamedRerouteUsage()
   Normal=@n23
   ClearCoatBottomNormal=@n23
   ShadingModel=ShadingModel()
  ) {attributeSetTypes [
    #1
    #2
    #3
    #4
    #5
    #6
    #7
   ]}
  FunctionOutput(None=@n24) {outputName=SetMaterialAttributes}
  FunctionOutput(None=NamedRerouteDeclaration(None=@n22.R) {name=CarbonFiberMask, nodeColor=<rgba 0.026 0.026 0.026 1>}) {outputName=CarbonFiberMask}
  FunctionOutput(None=NamedRerouteDeclaration(None=@n22.B) {name=Plastics, nodeColor=<rgba 0.026 0.026 0.026 1>}) {outputName=PlasticsMask}
  n25 = Saturate(None=Multiply(A=@n18) {constB=10})
  n26 = NamedRerouteUsage()
  LinearInterpolate(B=@n25, Alpha=@n26)
  n27 = Saturate(None=Multiply(A=@n19) {constB=10})
  n28 = NamedRerouteUsage()
  LinearInterpolate(B=@n27, Alpha=@n28)
  n29 = Saturate(None=Multiply(A=@n20) {constB=10})
  n30 = NamedRerouteUsage()
  LinearInterpolate(B=@n29, Alpha=@n30)
  n31 = Saturate(None=Multiply(A=@n21) {constB=10})
  n32 = NamedRerouteUsage()
  LinearInterpolate(B=@n31, Alpha=@n32)
  Max(B=@n1)
  n33 = TextureSampleParameter2D() {
   desc=_M
   group=Textures
   parameterName=Masks
   samplerSource=SSM_Wrap_WorldGroupSettings
   samplerType=SAMPLERTYPE_LinearColor
   sortPriority=33
   texture=&T_Black_Linear
   channelNames {a="", b=BoostMask, g=PaintMask, r=EmssivePartsMask}
  }
  NamedRerouteDeclaration(None=@n33.PaintMask) {name=ChassisPaintedMask, nodeColor=<rgba 0 0.1726 0.4792 1>}
  NamedRerouteDeclaration(None=VectorParameter() {group=Color, parameterName=Chassis_HeadlightColor, defaultValue=<rgba 0.6249 1 0.9901 1>}.RGB) {name=HeadlightColor, nodeColor=<rgba 0.9282 0 1 1>}
  n34 = TextureSampleParameter2D() {
   group=Textures
   parameterName=Normal
   samplerSource=SSM_Wrap_WorldGroupSettings
   samplerType=SAMPLERTYPE_Normal
   sortPriority=34
   texture=/VehicleCosmetics/Textures/T_BlankNormal_B
  }
  NamedRerouteDeclaration(None=@n34.RGB) {name=NormalTexture, nodeColor=<rgba 0.6038 0.3325 0.013 1>}
  NamedRerouteDeclaration(None=VectorParameter() {group=Color, parameterName=Chassis_BoostColor, defaultValue=<rgba 1 0.05 0 1>}.RGB) {name=BoostColor, nodeColor=<rgba 0.9282 0 1 1>}
  n35 = VectorParameter() {
   desc=Brightness Settings
   group=Emissive
   parameterName=Chassis_EmissiveBrightness
   sortPriority=0
   channelNames {a=BoostEmissive, b=BrakeLightEmissive, g=HeadLightEmissive, r=ChassisEmissiveParts}
   defaultValue=<rgba 1 1 1 1>
  }
  NamedRerouteDeclaration(None=@n35.ChassisEmissiveParts) {name=ChassisEmissivePartsBrightness, nodeColor=<rgba 0.0436 1 0 1>}
  NamedRerouteDeclaration(None=@n35.HeadLightEmissive) {name=HeadlightEmissiveBrightness, nodeColor=<rgba 0.0436 1 0 1>}
  NamedRerouteDeclaration(None=@n35.BrakeLightEmissive) {name=BrakelightEmissiveBrightness, nodeColor=<rgba 0.0436 1 0 1>}
  NamedRerouteDeclaration(None=@n35.BoostEmissive) {name=BoostEmissiveBrightness, nodeColor=<rgba 0.0436 1 0 1>}
  NamedRerouteDeclaration(None=ScalarParameter() {group=BP-USE-ONLY, parameterName=BoostGlowIntensity}) {name=BoostGlowVariable, nodeColor=<rgba 0 0 0 1>}
  NamedRerouteDeclaration(None=ScalarParameter() {group=BP-USE-ONLY, parameterName=isBraking}) {name=isBrakingVariable, nodeColor=<rgba 0 0 0 1>}
  NamedRerouteDeclaration(None=ScalarParameter() {group=BP-USE-ONLY, parameterName=ReverseLight}) {name=ReverseLightVariable, nodeColor=<rgba 0 0 0 1>}
  NamedRerouteDeclaration(None=ScalarParameter() {group=BP-USE-ONLY, parameterName=HeadlightsFullyOn}) {name=HeadlightsFullyOnVariable, nodeColor=<rgba 0 0 0 1>}
  n36 = TextureSampleParameter2D() {
   group=Textures
   parameterName=BaseColor
   samplerSource=SSM_Wrap_WorldGroupSettings
   sortPriority=30
   texture=/VehicleCosmetics/Textures/T_Black
  }
  NamedRerouteDeclaration(None=@n36.RGB) {name=BaseColorTexture, nodeColor=<rgba 0.6038 0.3325 0.013 1>}
  NamedRerouteDeclaration(None=@n33.BoostMask) {name=ChassisBoostEmissiveMask, nodeColor=<rgba 0 0.1726 0.4792 1>}
  MaterialExpressionReroute_1 = Reroute(None=@n3)
  MaterialExpressionReroute_9 = Reroute(None=@n4)
  n37 = TextureSample(Tex=@MaterialExpressionReroute_9.T2D_In) {samplerType=SAMPLERTYPE_LinearColor, texture=&T_Black_Linear}
  n38 = ComponentMask(None=@n37.RGBA) {g=true}
  n39 = ScalarParameter() {defaultValue=0.02, group=Emissive, parameterName=Headlight_OffsetRatio, sortPriority=52}
  n40 = ComponentMask(None=@n37.RGBA) {b=true}
  n41 = ScalarParameter() {defaultValue=0.03, group=Emissive, parameterName=Brakelight_OffsetRatio, sortPriority=54}
  n42 = Max(
   A=TextureSample(UVs=BumpOffset(Height=@n38, HeightRatioInput=@n39) {heightRatio=0.02, referencePlane=1}, Tex=@MaterialExpressionReroute_9.T2D_In) {samplerType=SAMPLERTYPE_LinearColor, texture=&T_Black_Linear}.G
   B=TextureSample(UVs=BumpOffset(Height=@n40, HeightRatioInput=@n41) {heightRatio=0.03, referencePlane=1}, Tex=@MaterialExpressionReroute_9.T2D_In) {samplerType=SAMPLERTYPE_LinearColor, texture=&T_Black_Linear}.B
  )
  n43 = TextureSample(
   UVs=BumpOffset(Height=@n37.R, HeightRatioInput=ScalarParameter() {group=Emissive, parameterName=Reverselight_OffsetRatio, sortPriority=56}) {heightRatio=0.03, referencePlane=1}
   Tex=@MaterialExpressionReroute_9.T2D_In
  ) {samplerType=SAMPLERTYPE_LinearColor, texture=&T_Black_Linear}
  n44 = StaticSwitchParameter(True=Multiply(A=Max(A=@n42, B=@n43.R), B=@n37.RGBA), False=@n37.RGBA) {group=Emissive, parameterName=BumpOffsetEmissive, sortPriority=50}
  n45 = StaticSwitchParameter(
   True=MaterialFunctionCall!(
    UVs
    TextureIn=@MaterialExpressionReroute_9.T2D_In
    UpperLeftCorner=ComponentMask(None=@MaterialExpressionReroute_1.RGBA) {g=true, r=true}
    LowerRightCorner=ComponentMask(None=@MaterialExpressionReroute_1.RGBA) {a=true, b=true}
    ExportFloat4
   ) {materialFunction=/Engine/Functions/Engine_MaterialFunctions01/Texturing/TextureCropping}.CroppedMasked
   False=QualitySwitch(Default=@n37.RGBA, High=@n44, Epic=@n44)
  ) {desc=CROP_EMIS, group=Emissive, parameterName=CroppedEmissive}
  n46 = ComponentMask(None=@n45) {g=true}
  MaterialExpressionReroute_13 = Reroute(None=@n46)
  NamedRerouteDeclaration(None=@MaterialExpressionReroute_13.HeadLight) {name=HeadlightEmissiveMask, nodeColor=<rgba 0 0.1726 0.4792 1>}
  n47 = ComponentMask(None=@n45) {b=true}
  MaterialExpressionReroute_14 = Reroute(None=@n47)
  NamedRerouteDeclaration(None=@MaterialExpressionReroute_14.BrakeLight) {name=BrakeLightEmissiveMask, nodeColor=<rgba 0 0.1726 0.4792 1>}
  n48 = ComponentMask(None=@n45) {r=true}
  MaterialExpressionReroute_0 = Reroute(None=@n48)
  NamedRerouteDeclaration(None=@MaterialExpressionReroute_0.ReverseLight) {name=ReverseLightEmissiveMask, nodeColor=<rgba 0 0.1726 0.4792 1>}
  n49 = NamedRerouteUsage()
  n50 = Multiply(A=NamedRerouteUsage(), B=Multiply(A=@n49, B=NamedRerouteUsage())) {desc=HeadLight Color and Brightness}
  n51 = Multiply(A=Saturate(None=ConstantBiasScale(None=@n49) {bias=0.1, scale=1.5}), B=@n50) {desc="Enhances BC emissive so it doesn't fade to white.\r\nDoes slightly affect Headlight Color if it's enabled."}
  n52 = StaticSwitchParameter(True=Constant() {r=1}, False=Constant()) {desc=BC overrides emissive color while true., group=Color, parameterName=UseBC_HeadlightEmissive}
  NamedRerouteDeclaration(None=Multiply(A=NamedRerouteUsage(), B=LinearInterpolate(A=@n50, B=@n51, Alpha=@n52) {desc="Push color of BC without\r\naffecting any other settings."}) {desc=HeadLight Variable Mask}) {name=HeadLightCombined, nodeColor=<rgba 1 1 1 1>}
  n53 = NamedRerouteUsage()
  n54 = Multiply(A=NamedRerouteUsage(), B=Multiply(A=@n53, B=NamedRerouteUsage())) {desc=HeadLight Color and Brightness}
  n55 = Multiply(A=Saturate(None=ConstantBiasScale(None=@n53) {bias=0.1, scale=1.5}), B=@n54) {desc="Enhances BC emissive so it doesn't fade to white.\r\nDoes slightly affect Headlight Color if it's enabled."}
  n56 = StaticSwitchParameter(True=Constant() {r=1}, False=Constant()) {desc=BC overrides emissive color while true., group=Color, parameterName=UseBC_BrakeEmissive}
  NamedRerouteDeclaration(None=Multiply(A=NamedRerouteUsage(), B=LinearInterpolate(A=@n54, B=@n55, Alpha=@n56) {desc="Push color of BC without\r\naffecting any other settings."}) {desc=HeadLight Variable Mask}) {name=BrakeLightCombined, nodeColor=<rgba 1 1 1 1>}
  n57 = NamedRerouteUsage()
  n58 = Multiply(A=Constant() {r=1}, B=NamedRerouteUsage())
  NamedRerouteDeclaration(None=Multiply(A=NamedRerouteUsage(), B=Multiply(A=@n57, B=@n58) {desc=HeadLight Color and Brightness})) {name=ReverseLightCombined, nodeColor=<rgba 1 1 1 1>}
  n59 = NamedRerouteUsage()
  n60 = Multiply(A=NamedRerouteUsage(), B=NamedRerouteUsage())
  NamedRerouteDeclaration(None=Multiply(A=NamedRerouteUsage(), B=Multiply(A=@n59, B=@n60) {desc=HeadLight Color and Brightness})) {name=BoostCombined, nodeColor=<rgba 1 1 1 1>}
  NamedRerouteDeclaration(None=@n33.EmssivePartsMask) {name=ChassisEmissivePartsMask, nodeColor=<rgba 0 0.1726 0.4792 1>}
  n61 = VectorParameter() {
   desc="The Glowing parts of the emissive (if any) controlled by the EmissivePartsMask. --"
   group=Color
   parameterName=Chassis_EmissivePartsColor
   defaultValue=<rgba 0.6038 0.3325 0.013 1>
  }
  NamedRerouteDeclaration(None=@n61.RGB) {name=ChassisEmissivePartsColor, nodeColor=<rgba 0.9301 0 1 1>}
  n62 = VectorParameter() {
   desc="The Painted Color controlled by the PaintMask. -- Used for the paintable areas of the chassis --\r\nSetting alpha to 0 will bypass colorization and show the basecolor texture instead; useful for painted trims that need BC texture for None"
   group=Color
   parameterName=Chassis_PaintColor
   defaultValue=<rgba 0 0.1434 1 1>
  }
  NamedRerouteDeclaration(None=@n62.RGB) {name=PaintColor, nodeColor=<rgba 0.9282 0 1 1>}
  n63 = NamedRerouteUsage()
  n64 = VectorParameter() {
   desc=BC overrides emissive color when set to 1
   group=Emissive
   parameterName=UseBC_ReplaceEC
   channelNames {a=BoostEmissive, b=BrakeLightEmissive, g=HeadLightEmissive, r=ChassisEmissiveParts}
  }
  n65 = Floor(None=@n64.BrakeLightEmissive)
  NamedRerouteDeclaration(None=LinearInterpolate(A=NamedRerouteUsage(), B=@n63, Alpha=Saturate(None=@n65))) {name=EC_BrakeEmis_Bool, nodeColor=<rgba 0.3763 0.7529 0.7157 1>}
  n66 = Add(A=Add(A=NamedRerouteUsage(), B=NamedRerouteUsage()), B=NamedRerouteUsage())
  n67 = NamedRerouteUsage()
  NamedRerouteDeclaration(None=Add(A=Add(A=@n66, B=@n67), B=NamedRerouteUsage())) {name=EmissiveColor, nodeColor=<rgba 0 0.7656 1 1>}
  n68 = NamedRerouteUsage()
  n69 = NamedRerouteUsage()
  NamedRerouteDeclaration(None=Multiply(A=NamedRerouteUsage(), B=Multiply(A=@n68, B=@n69))) {name=ChassisEmissivePartsCombined, nodeColor=<rgba 1 1 1 1>}
  n70 = Floor(None=@n64.BoostEmissive)
  NamedRerouteDeclaration(None=LinearInterpolate(A=NamedRerouteUsage(), B=@n63, Alpha=Saturate(None=@n70))) {name=EC_BoostEmis_Bool, nodeColor=<rgba 0.3763 0.7529 0.7157 1>}
  MaterialExpressionReroute_12 = Reroute(None=@n5)
  MaterialExpressionReroute_15 = Reroute(None=@n17)
  n71 = ScalarParameter() {
   desc="When set to 1, uses the old method for using BC in the emissive. SHOULD NOT BE USED FOR NON-LEGACY CONTENT!"
   group=Emissive
   parameterName=UseBC_LegecyColor
   sortPriority=33
  }
  n72 = Floor(None=@n71)
  n73 = LinearInterpolate(
   A=LinearInterpolate(A=@MaterialExpressionReroute_12.BaseColorPainted, B=@MaterialExpressionReroute_15.BaseColorEmissive, Alpha=Saturate(None=@n72))
   B=@MaterialExpressionReroute_15.BaseColorEmissive
   Alpha=Multiply(A=@n25, B=@n26)
  )
  n74 = Multiply(A=@n31, B=@n32)
  n75 = LinearInterpolate(
   A=LinearInterpolate(
    A=LinearInterpolate(A=@n73, B=@MaterialExpressionReroute_15.BaseColorEmissive, Alpha=@n74)
    B=@MaterialExpressionReroute_15.BaseColorEmissive
    Alpha=Multiply(A=@n29, B=@n30)
   )
   B=@MaterialExpressionReroute_15.BaseColorEmissive
   Alpha=Multiply(A=@n27, B=@n28)
  )
  MaterialExpressionReroute_16 = Reroute(None=@n75)
  NamedRerouteDeclaration(None=@MaterialExpressionReroute_16.BaseColorFinal) {name=BaseColorFinal, nodeColor=<rgba 1 0.6328 0 1>}
  n76 = Floor(None=@n64.HeadLightEmissive)
  NamedRerouteDeclaration(None=LinearInterpolate(A=NamedRerouteUsage(), B=@n63, Alpha=Saturate(None=@n76))) {name=EC_HeadlightEmiss_Bool, nodeColor=<rgba 0.3763 0.7529 0.7157 1>}
  NamedRerouteDeclaration(None=VectorParameter() {group=Color, parameterName=Brake_Color, defaultValue=<rgba 0.6 0 0 1>}.RGB) {name=ChassisBrakeColor, nodeColor=<rgba 0.9301 0 1 1>}
  n77 = Floor(None=@n64.ChassisEmissiveParts)
  NamedRerouteDeclaration(None=LinearInterpolate(A=NamedRerouteUsage(), B=@n63, Alpha=Saturate(None=@n77))) {name=EC_ChassisPartsEmis_Bool, nodeColor=<rgba 0.3763 0.7529 0.7157 1>}
  n78 = VectorParameter() {
   desc=Emissive color overrides base color when set to 1
   group=Color
   parameterName=UseEC_ReplaceBC
   channelNames {a=BoostEmissive, b=BrakeLightEmissive, g=HeadLightEmissive, r=ChassisEmissiveParts}
  }
  n79 = Saturate(None=Floor(None=@n78.ChassisEmissiveParts))
  n80 = Saturate(None=NamedRerouteUsage())
  n81 = VectorParameter() {
   desc="When set to 1, the emissive color will only replace the base color if the emissive group is active (e.g., the headlight is on)."
   group=Color
   parameterName=UseEC_ReplaceOnlyWhenActive
   sortPriority=33
   channelNames {a=BoostEmissive, b=BrakeLightEmissive, g=HeadLightEmissive, r=ChassisEmissiveParts}
   defaultValue=<rgba 1 1 1 1>
  }
  n82 = Floor(None=@n81.ChassisEmissiveParts)
  NamedRerouteDeclaration(None=LinearInterpolate(A=@n79, B=Multiply(A=@n80, B=@n79), Alpha=Saturate(None=@n82))) {name=EC_ReplaceBC_EmissiveParts, nodeColor=<rgba 0.875 0 1 1>}
  n83 = Floor(None=@n78.HeadLightEmissive)
  n84 = Floor(None=@n81.HeadLightEmissive)
  NamedRerouteDeclaration(None=LinearInterpolate(A=Saturate(None=@n83), B=NamedRerouteUsage(), Alpha=Saturate(None=@n84))) {name=EC_ReplaceBC_HeadLight, nodeColor=<rgba 0.875 0 1 1>}
  n85 = Floor(None=@n78.BoostEmissive)
  n86 = Floor(None=@n81.BoostEmissive)
  NamedRerouteDeclaration(None=LinearInterpolate(A=Saturate(None=@n85), B=NamedRerouteUsage(), Alpha=Saturate(None=@n86))) {name=EC_ReplaceBC_Boost, nodeColor=<rgba 0.875 0 1 1>}
  n87 = Floor(None=@n78.BrakeLightEmissive)
  n88 = Floor(None=@n81.BrakeLightEmissive)
  NamedRerouteDeclaration(None=LinearInterpolate(A=Saturate(None=@n87), B=NamedRerouteUsage(), Alpha=Saturate(None=@n88))) {name=EC_ReplaceBC_BrakeLight, nodeColor=<rgba 0.875 0 1 1>}
  n89 = NamedRerouteUsage()
  NamedRerouteDeclaration(None=LinearInterpolate(A=@n89, B=NamedRerouteUsage(), Alpha=NamedRerouteUsage())) {name=BC_ChassisPartsEmis_Bool, nodeColor=<rgba 0.3763 0.7529 0.7157 1>}
  NamedRerouteDeclaration(None=LinearInterpolate(A=@n89, B=NamedRerouteUsage(), Alpha=NamedRerouteUsage())) {name=BC_HeadlightEmiss_Bool, nodeColor=<rgba 0.3763 0.7529 0.7157 1>}
  NamedRerouteDeclaration(None=LinearInterpolate(A=@n89, B=NamedRerouteUsage(), Alpha=NamedRerouteUsage())) {name=BC_BrakeEmis_Bool, nodeColor=<rgba 0.3763 0.7529 0.7157 1>}
  NamedRerouteDeclaration(None=LinearInterpolate(A=@n89, B=NamedRerouteUsage(), Alpha=NamedRerouteUsage())) {name=BC_BoostEmis_Bool, nodeColor=<rgba 0.3763 0.7529 0.7157 1>}
  n90 = NamedRerouteUsage()
  n91 = Saturate(None=Ceil(None=ScalarParameter() {defaultValue=1, group=Color, parameterName=Chassis_PaintColor_UseVectorColor}))
  NamedRerouteDeclaration(None=LinearInterpolate(A=NamedRerouteUsage(), B=NamedRerouteUsage(), Alpha=Multiply(A=@n90, B=@n91)) {desc=Painted Mask}) {name=BaseColorPainted, nodeColor=<rgba 0 0.7656 1 1>}
  n92 = TextureSampleParameter2D() {
   group=Textures
   parameterName=ChassisTextures
   samplerSource=SSM_Wrap_WorldGroupSettings
   samplerType=SAMPLERTYPE_LinearColor
   sortPriority=35
   texture=&T_Black_Linear
   channelNames {a=Roughness, b=Metallic, g=Empty, r=Empty}
  }
  NamedRerouteDeclaration(None=@n92.Roughness) {name=Roughness, nodeColor=<rgba 0 0.1726 0.4792 1>}
  NamedRerouteDeclaration(None=@n92.Metallic) {name=Metallic, nodeColor=<rgba 0 0.1726 0.4792 1>}
  NamedRerouteUsage()
  NamedRerouteUsage()
  NamedRerouteUsage()
  NamedRerouteUsage()
  NamedRerouteUsage()
  PinBase() {reroutePins [
    RGBA {expression=&self:MaterialExpressionReroute_1}
    T2D_In {expression=&self:MaterialExpressionReroute_9}
   ]}
  PinBase!(ReverseLight=@n48, HeadLight=@n46, BrakeLight=@n47) {reroutePins [
    ReverseLight {expression=&self:MaterialExpressionReroute_0}
    HeadLight {expression=&self:MaterialExpressionReroute_13}
    BrakeLight {expression=&self:MaterialExpressionReroute_14}
   ]}
  PinBase!()
  PinBase!(EmissiveColors=@n14, EmissiveMasks=@n16) {reroutePins [
    EmissiveColors {expression=&self:MaterialExpressionReroute_2}
    EmissiveMasks {expression=&self:MaterialExpressionReroute_3}
   ]}
  PinBase!() {reroutePins [
    BaseColorPainted {expression=&self:MaterialExpressionReroute_12}
    BaseColorEmissive {expression=&self:MaterialExpressionReroute_15}
   ]}
  PinBase!(BaseColorFinal=@n75) {reroutePins [BaseColorFinal {expression=&self:MaterialExpressionReroute_16}]}
  n93 = NamedRerouteUsage()
  StaticSwitchParameter(True=@n93, False=NamedRerouteUsage()) {desc=BC overrides emissive color while true., group=Color, parameterName=UseBC_BoostEmissive}
  StaticSwitchParameter(True=@n93, False=NamedRerouteUsage()) {desc=BC overrides emissive color while true., group=Color, parameterName=UseBC_HeadlightEmissive}
  StaticSwitchParameter(True=@n93, False=NamedRerouteUsage()) {desc=BC overrides emissive color while true., group=Color, parameterName=UseBC_ChassisPartsEmissive}
  StaticSwitchParameter(True=@n93, False=NamedRerouteUsage()) {desc=BC overrides emissive color while true., group=Color, parameterName=UseBC_BrakeEmissive}
 ]}
