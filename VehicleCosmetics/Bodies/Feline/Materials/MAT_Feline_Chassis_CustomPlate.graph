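# Chassis material for the Feline vehicle body that adds a custom license plate over MF_Chassis_Base_v2.
# A CustomLicensePlate subgraph crops a mask texture and blends plate base color, metallic and roughness into the base attributes.
# Plate background and text colors derive from the chassis paint via square and square-root curves with clamps, or use unique color parameters.
# Static switches then layer trim, carbon fiber and plastics, with DeRez-transitioned attributes feeding clear coat bottom normal.
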Material /VehicleCosmetics/Bodies/Feline/Materials/MAT_Feline_Chassis_CustomPlate
graph {
 properties {
  bUseMaterialAttributes=true
  bUsedWithMeshDeformer=true
  bUsedWithSkeletalMesh=true
  pixelDepthOffsetMode=PDOM_Legacy
  shadingModel=MSM_FromMaterialExpression
  usedShadingModels=Default Lit | Clear Coat
 }
 nodes [
  n1 = NamedRerouteUsage()
  n2 = MaterialFunctionCall() {materialFunction=/VehicleCosmetics/Materials/VehicleParents_Version2/Material_Functions/MF_Chassis_Base_v2}
  MaterialExpressionReroute_5 = Reroute(None=@n2.SetMaterialAttributes)
  n3 = GetMaterialAttributes(None=@MaterialExpressionReroute_5."In (MA)") {attributeGetTypes [#1, #2, #3]}
  n4 = SetMaterialAttributes!(
   MaterialAttributes=@n3.MaterialAttributes
   Base Color=LinearInterpolate(A=@n3.Base Color, B=NamedRerouteUsage(), Alpha=NamedRerouteUsage())
   Metallic=LinearInterpolate(A=@n3.Metallic, B=NamedRerouteUsage(), Alpha=NamedRerouteUsage())
   Roughness=LinearInterpolate(A=@n3.Roughness, B=NamedRerouteUsage(), Alpha=NamedRerouteUsage())
  ) {attributeSetTypes [#1, #2, #3]}
  MaterialExpressionReroute_6 = Reroute(None=@n4)
  n5 = MaterialFunctionCall!(MaterialAttributes=@MaterialExpressionReroute_6."Out (MA)", INTrimMask=NamedRerouteUsage()) {materialFunction=/VehicleCosmetics/Materials/VehicleParents_Version2/Material_Functions/MF_Apply_Trim_v2}
  n6 = MaterialFunctionCall(MaterialAttributes=@n5.SetMaterialAttributes, INTrimMask=@n1) {materialFunction=&MF_Apply_Trim_v2}
  n7 = MaterialFunctionCall!(MaterialAttributes=@n5.SetMaterialAttributes, INCarbonFIberMask=@n1) {materialFunction=/VehicleCosmetics/Materials/MaterialFunctions/MF_Apply_CarbonFiber}
  n8 = ScalarParameter() {parameterName=Trim_SwapTrimForCarbonFiber}
  n9 = StaticSwitchParameter(
   True=StaticSwitchParameter(True=BlendMaterialAttributes(A=@n6.SetMaterialAttributes, B=@n7.SetMaterialAttributes, Alpha=@n8), False=@n5.SetMaterialAttributes) {group=SpecialFeatures, parameterName=TrimSwapForCarbonFiber}
   False=@MaterialExpressionReroute_6."Out (MA)"
  ) {group=Trim, parameterName=ApplyTrim}
  n10 = StaticSwitchParameter(
   True=MaterialFunctionCall(MaterialAttributes=@n9, INCarbonFIberMask=NamedRerouteUsage()) {materialFunction=&MF_Apply_CarbonFiber}.SetMaterialAttributes
   False=@n9
  ) {group=CarbonFiber, parameterName=ApplyCarbonFiber}
  n11 = StaticSwitchParameter(
   True=MaterialFunctionCall!(MaterialAttributes=@n10, INPlasticMask=NamedRerouteUsage()) {materialFunction=/VehicleCosmetics/Materials/MaterialFunctions/MF_Apply_Plastics}.SetMaterialAttributes
   False=@n10
  ) {group=Plastic, parameterName=ApplyPlastics}
  ClearCoatNormalCustomOutput!(None=GetMaterialAttributes(None=MaterialFunctionCall(In=@n11) {materialFunction=&MF_VehicleDeRez_Transition}.Result) {attributeGetTypes [#4]}.ClearCoatBottomNormal)
  Composite!("In (MA)"=@n2.SetMaterialAttributes) {subgraphName=CustomLicensePlate}
  Constant2Vector() {r=0.5576}
  Constant2Vector() {g=0.0571, r=0.6963}
  Constant() {desc=.925*.925, r=0.8556}
  NamedRerouteDeclaration(None=@n2.TrimMask) {name=ChassisBase_TrimMask, nodeColor=<rgba 0 1 0.1797 1>}
  NamedRerouteDeclaration(None=@n2.CarbonFiberMask) {name=ChassisBase_CarbonFiberMask, nodeColor=<rgba 0.6641 0 1 1>}
  n12 = Constant() {desc=.90*.90, r=0.81}
  n13 = NamedRerouteUsage()
  n14 = Multiply(A=@n13, B=@n13)
  n15 = NamedRerouteUsage()
  n16 = Desaturation(None=Ceil(None=Subtract(A=@n15) {constB=0.9})) {luminanceFactors=<rgba 1 1 1 1>}
  NamedRerouteDeclaration(None=LinearInterpolate(A=NamedRerouteUsage(), B=Multiply(A=@n12, B=@n14) {desc=Clamp colors above 0.90}, Alpha=Saturate(None=@n16))) {name=PlateColor_CompiledBackground, nodeColor=<rgba 0.2188 0 1 1>}
  n17 = NamedRerouteUsage()
  n18 = Multiply(A=Multiply(A=SquareRoot(None=@n17)) {constB=2}, B=Constant() {desc="Sqrt(0.0015)*2", r=0.0775}) {desc=Clamp colors starting at 0.005}
  n19 = NamedRerouteUsage()
  n20 = Saturate(None=Desaturation(None=Ceil(None=@n19)) {luminanceFactors=<rgba 1 1 1 1>})
  n21 = SquareRoot(None=NamedRerouteUsage())
  n22 = NamedRerouteUsage()
  n23 = Desaturation(None=Ceil(None=Subtract(A=@n22) {constB=0.005})) {luminanceFactors=<rgba 1 1 1 1>}
  NamedRerouteDeclaration(None=LinearInterpolate(A=LinearInterpolate(B=@n18, Alpha=@n20) {constA=0.15}, B=Multiply(A=@n21) {constA=2, constB=2}, Alpha=Saturate(None=@n23))) {name=PlateColor_CompiledText, nodeColor=<rgba 0 0.5078 1 1>}
  n24 = VectorParameter() {group=CUSTOM_PLATE, parameterName=CustomPlateUVs_RG_UpperLeftCorner_BA_LowerRightCorner, defaultValue=<rgba 0.5576 0 0.6963 0.0571>}
  n25 = MaterialFunctionCall!(
   UVs=TextureCoordinate()
   TextureIn=TextureObject() {samplerType=SAMPLERTYPE_LinearColor, texture=/VehicleCosmetics/Bodies/Feline/Textures/T_Feline_LicensePlate_M}
   UpperLeftCorner=AppendVector(A=@n24.R, B=@n24.R)
   LowerRightCorner=AppendVector(A=@n24.B, B=@n24.B)
   ExportFloat4
  ) {materialFunction=/Engine/Functions/Engine_MaterialFunctions01/Texturing/TextureCropping}
  NamedRerouteDeclaration(None=StaticSwitchParameter(True=ComponentMask(None=@n25.CroppedMasked) {b=true}, False=@n25.Crop Mask) {group=CUSTOM_PLATE, parameterName=UseLODMask}) {name=CropMask, nodeColor=<rgba 1 0 0.3516 1>}
  n26 = StaticBoolParameter() {group=CUSTOM_PLATE, parameterName=UseUniquePlateColors}
  n27 = StaticSwitch(
   True=VectorParameter() {group=CUSTOM_PLATE, parameterName=LicensePlateBackground, defaultValue=<rgba 0.0057 0 0.9779 0>}.RGB
   False=NamedRerouteUsage()
   Value=@n26
  )
  n28 = StaticSwitch(
   True=VectorParameter() {group=CUSTOM_PLATE, parameterName=LicensePlateText, defaultValue=<rgba 1 0.9323 0 0>}.RGB
   False=NamedRerouteUsage()
   Value=@n26
  )
  n29 = NamedRerouteUsage()
  NamedRerouteDeclaration(None=Saturate(None=LinearInterpolate(A=@n27, B=@n28, Alpha=@n29))) {name=PlateBC_OUT, nodeColor=<rgba 0 0.6719 1 1>}
  NamedRerouteDeclaration(None=StaticSwitchParameter(True=ComponentMask(None=@n25.CroppedMasked) {r=true}, False=ComponentMask(None=@n25.CroppedMasked) {g=true}) {group=CUSTOM_PLATE, parameterName=UseLODMask}) {name=CroppedMask, nodeColor=<rgba 0.9922 1 0 1>}
  NamedRerouteDeclaration(None=LinearInterpolate(
    A=ScalarParameter() {defaultValue=1, group=CUSTOM_PLATE, parameterName=PlateBackground_Metallic}
    B=ScalarParameter() {defaultValue=1, group=CUSTOM_PLATE, parameterName=PlateText_Metallic}
    Alpha=NamedRerouteUsage()
   )) {name=PlateMetallic_OUT, nodeColor=<rgba 0 1 0.2031 1>}
  NamedRerouteDeclaration(None=LinearInterpolate(
    A=ScalarParameter() {defaultValue=0.4, group=CUSTOM_PLATE, parameterName=PlateBackground_Roughness}
    B=ScalarParameter() {defaultValue=0.45, group=CUSTOM_PLATE, parameterName=PlateText_Roughness}
    Alpha=NamedRerouteUsage()
   )) {name=PlateRoughness_OUT, nodeColor=<rgba 0 1 0.2031 1>}
  NamedRerouteDeclaration(None=@n2.PlasticsMask) {name=ChassisBase_PlasticsMask, nodeColor=<rgba 1 0 0.6797 1>}
  n30 = VectorParameter() {
   desc=The Painted Color controlled by the PaintMask. -- Used for the paintable areas of the chassis --
   group=Color
   parameterName=Chassis_PaintColor
   defaultValue=<rgba 0 0.1434 1 1>
  }
  n31 = Max(
   A=Max(
    A=NamedRerouteDeclaration(None=@n30.R) {name=PC_Red, nodeColor=<rgba 1 0.9609 0 1>}
    B=NamedRerouteDeclaration(None=@n30.G) {name=PC_Green, nodeColor=<rgba 1 0.9609 0 1>}
   )
   B=NamedRerouteDeclaration(None=@n30.B) {name=PC_Blue, nodeColor=<rgba 1 0.9609 0 1>}
  )
  n32 = LinearInterpolate(B=@n31, Alpha=Saturate(None=Ceil(None=@n31))) {constA=-1, desc=Ensures we don't divide by 0}
  n33 = AppendVector(A=Divide(A=NamedRerouteUsage(), B=@n32), B=Divide(A=NamedRerouteUsage(), B=@n32))
  n34 = Divide(A=NamedRerouteUsage(), B=@n32)
  NamedRerouteDeclaration(None=Saturate(None=AppendVector(A=@n33, B=@n34))) {name=PlateColor_Ratio, nodeColor=<rgba 0.5 1 0 1>}
  NamedRerouteDeclaration(None=Multiply(A=@n30.RGB, B=@n30.RGB)) {name=PlateColor_OriginalBackground, nodeColor=<rgba 1 0 0.8203 1>}
  NamedRerouteDeclaration(None=@n30.RGB) {name=PlateColor_Raw, nodeColor=<rgba 1 0.5391 0 1>}
  PinBase() {reroutePins ["In (MA)" {expression=&self:MaterialExpressionReroute_5}]}
  PinBase!("Out (MA)"=@n4) {reroutePins ["Out (MA)" {expression=&self:MaterialExpressionReroute_6}]}
  SetMaterialAttributes!(
   MaterialAttributes=@n7.SetMaterialAttributes
   ClearCoatBottomNormal=GetMaterialAttributes(None=@n5.SetMaterialAttributes) {attributeGetTypes [#4]}.ClearCoatBottomNormal
  ) {attributeSetTypes [#4]}
  StaticComponentMaskParameter(None=@n25.CroppedMasked) {defaultG=true}
  Subtract(A=@n15) {constB=0.925}
 ]
}
