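# Rear light material for the Feline vehicle body, with brake and reverse colors driven by Blueprint-set isBraking and ReverseLight parameters.
# Cropped, mirrored UVs and bump offsets sample the tail-light mask texture to shape the light face.
# Tangent-space camera parallax and procedural sine and radial patterns fake interior lights, dividers and dotted detail, with quality switches simplifying lower tiers.
# Named reroutes carry results into material attributes, passed through MF_Vehicle_FNPostFX.
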Material /VehicleCosmetics/Bodies/Feline/Materials/MAT_Feline_RearLights
graph {
 properties {
  bAutomaticallySetUsageInEditor=false
  bUseMaterialAttributes=true
  bUsedWithMeshDeformer=true
  bUsedWithSkeletalMesh=true
  usedShadingModels=Default Lit
 }
 nodes [
  n1 = NamedRerouteUsage()
  n2 = LinearInterpolate(Alpha=@n1) {constA=0.2}
  n3 = NamedRerouteUsage()
  n4 = LinearInterpolate(Alpha=Saturate(None=ConstantBiasScale(None=@n3) {bias=0, scale=2})) {constA=0.2, constB=0.5}
  n5 = ConstantBiasScale(None=@n3) {bias=-0.5, scale=1.5}
  n6 = Saturate(None=Add(A=Multiply(A=@n2, B=@n4), B=Saturate(None=@n5)))
  n7 = LinearInterpolate(A=Max(A=Multiply(A=@n2) {constB=0.8}, B=@n6), B=@n6, Alpha=LinearInterpolate(Alpha=NamedRerouteUsage()) {constA=0.2, constB=0.8}) {constAlpha=0}
  n8 = Max(A=@n7, B=NamedRerouteUsage())
  n9 = Multiply(A=Saturate(None=Power(Base=@n3) {constExponent=1.6}), B=Multiply(A=NamedRerouteUsage()) {constB=0.75})
  Add(A=@n8, B=@n9)
  Clamp(None=NamedRerouteUsage()) {minDefault=0.5}
  Constant2Vector() {g=-1, r=-1.5}
  Constant2Vector() {g=20, r=3}
  Constant3Vector() {constant=<rgba 0.0217 0.0217 0.0217 0>}
  ConstantBiasScale(None=@n1) {bias=-0.15, scale=1.5}
  n10 = Constant4Vector() {constant=<rgba 0.0098 0.0903 0.0776 0.1582>}
  n11 = ComponentMask(None=@n10.RGBA) {g=true, r=true}
  n12 = ComponentMask(None=@n10.RGBA) {a=true, b=true}
  n13 = MaterialFunctionCall!(UVs=TextureCoordinate(), TextureIn, UpperLeftCorner=@n11, LowerRightCorner=@n12, ExportFloat4) {materialFunction=/Engine/Functions/Engine_MaterialFunctions01/Texturing/TextureCropping}
  n14 = LinearInterpolate(Alpha=Saturate(None=TextureSample(UVs=@n13.Crop UVs) {samplerType=SAMPLERTYPE_LinearColor, texture=/VehicleCosmetics/Bodies/Feline/Textures/T_Feline_TailLight_M}.A)) {constA=0.003, constB=0.0055}
  n15 = NamedRerouteUsage()
  n16 = TextureCoordinate()
  n17 = QualitySwitch(Default=Add(A=Multiply(A=@n14, B=@n15), B=@n16), Low=@n16)
  n18 = MaterialFunctionCall(UVs=@n17, UpperLeftCorner=@n11, LowerRightCorner=@n12) {materialFunction=&TextureCropping}
  n19 = Constant4Vector() {constant=<rgba 0.2358 0.0903 0.3037 0.1582>}
  n20 = MaterialFunctionCall(UVs=@n17, UpperLeftCorner=ComponentMask(None=@n19.RGBA) {g=true, r=true}, LowerRightCorner=ComponentMask(None=@n19.RGBA) {a=true, b=true}) {materialFunction=&TextureCropping}
  n21 = Constant2Vector() {g=1, r=-1}
  n22 = Add(A=@n18.Crop UVs) {constB=-2}
  n23 = Saturate(None=Ceil(None=ComponentMask(None=@n22) {r=true}))
  n24 = LinearInterpolate(A=@n18.Crop UVs, B=Multiply(A=@n20.Crop UVs, B=@n21), Alpha=@n23)
  n25 = TextureSample(UVs=@n24) {samplerType=SAMPLERTYPE_LinearColor, texture=&T_Feline_TailLight_M}
  LinearInterpolate(Alpha=@n25.G) {constA=-1.5, constB=2.5}
  LinearInterpolate() {constA=0.002, constB=0.0065}
  n26 = NamedRerouteUsage()
  n27 = NamedRerouteUsage()
  n28 = NamedRerouteUsage()
  n29 = SetMaterialAttributes!(
   MaterialAttributes
   Base Color=NamedRerouteUsage()
   Metallic=Constant() {r=0.9}
   Specular=Constant() {r=0.3}
   Roughness=Constant() {r=0.1}
   Emissive Color=Multiply(A=LinearInterpolate(A=@n26, B=@n27, Alpha=@n28), B=NamedRerouteUsage())
   Normal=NamedRerouteUsage()
  ) {attributeSetTypes [
    #1
    #2
    #3
    #4
    #5
    #6
   ]}
  MaterialFunctionCall!(In=@n29) {materialFunction=/VehicleCosmetics/Materials/MaterialFunctions/MF_Vehicle_FNPostFX}
  n30 = LinearInterpolate(Alpha=@n25.G) {constA=-3, constB=5}
  n31 = Constant() {r=0.0125}
  n32 = BumpOffset(Coordinate=@n20.Crop UVs, Height=@n30, HeightRatioInput=@n31) {heightRatio=0.0075}
  n33 = TextureSample(UVs=LinearInterpolate(A=BumpOffset(Coordinate=@n24, Height=@n30, HeightRatioInput=@n31) {heightRatio=0.0075}, B=Multiply(A=@n32, B=@n21), Alpha=@n23)) {samplerType=SAMPLERTYPE_LinearColor, texture=&T_Feline_TailLight_M}
  n34 = Subtract(A=Saturate(None=AppendVector(A=@n33.R, B=@n25.R))) {constB=0.15}
  n35 = MaterialFunctionCall(Float2=Saturate(None=Multiply(A=@n34) {constB=1.1765})) {materialFunction=&BreakOutFloat2Components}
  n36 = Saturate(None=@n35.G)
  n37 = Subtract(A=Saturate(None=AppendVector(A=@n33.B, B=@n25.B))) {constB=0.15}
  n38 = MaterialFunctionCall(Float2=Saturate(None=Multiply(A=@n37) {constB=1.1765})) {materialFunction=&BreakOutFloat2Components}
  n39 = Saturate(None=@n38.G)
  Max(A=@n36, B=@n39)
  Multiply()
  n40 = Multiply(A=Add(A=@n7, B=@n9)) {constB=0.7}
  n41 = ConstantBiasScale(None=Desaturation(None=NamedRerouteUsage())) {bias=0.05, scale=2}
  n42 = NamedRerouteUsage()
  Multiply(A=Saturate(None=Multiply(A=@n40, B=@n41)), B=@n42)
  n43 = ComponentMask(None=@n8) {r=true}
  n44 = ComponentMask(None=@n7) {r=true}
  n45 = Multiply(A=Add(A=AppendVector(A=@n43, B=@n44), B=@n9) {constB=0.33}) {constB=0.7}
  n46 = Multiply(A=Saturate(None=Multiply(A=@n45, B=@n41)), B=@n42)
  NamedRerouteDeclaration(None=ComponentMask(None=@n46) {g=true}) {name=BaseColorEmissive, nodeColor=<rgba 0.5703 1 0 1>}
  n47 = Multiply(B=NamedRerouteUsage()) {constA=0.25}
  NamedRerouteDeclaration(None=Add(A=Constant() {r=0.05}, B=Add(A=@n47) {constB=0.5})) {name=ScannerOverlay, nodeColor=<rgba 0 1 0.6484 1>}
  NamedRerouteDeclaration(None=Constant3Vector() {constant=<rgba 0 0 1 0>}) {name=Normal_Out, nodeColor=<rgba 0.0078 0 1 1>}
  NamedRerouteDeclaration(None=ComponentMask(None=@n46) {r=true}) {name=LightsEmissive, nodeColor=<rgba 0 1 0.8828 1>}
  n48 = Transform(None=CameraVectorWS()) {transformSourceType=TRANSFORMSOURCE_World, transformType=TRANSFORM_Tangent}
  n49 = ComponentMask(None=@n48) {g=true, r=true}
  NamedRerouteDeclaration(None=Divide(A=Multiply(A=@n49) {constB=-1}, B=ComponentMask(None=@n48) {b=true}) {constB=1}) {name=Parallax, nodeColor=<rgba 0 1 0.2969 1>}
  n50 = Multiply(A=ComponentMask(None=NamedRerouteUsage()) {g=true, r=true}) {constB=0.5}
  n51 = Add(A=NamedRerouteUsage(), B=@n50)
  n52 = Multiply(A=Constant2Vector() {g=1, r=0.5}, B=@n51)
  n53 = LinearInterpolate(A=@n52, B=@n51, Alpha=NamedRerouteUsage())
  n54 = LinearInterpolate(A=Add(A=@n52, B=Constant2Vector() {g=1, r=0.5}), B=@n51, Alpha=NamedRerouteUsage())
  n55 = NamedRerouteUsage()
  n56 = Sine(None=Frac(None=LinearInterpolate(A=@n53, B=@n54, Alpha=@n55))) {period=2}
  n57 = ComponentMask(None=@n56) {g=true}
  n58 = ComponentMask(None=@n56) {r=true}
  n59 = LinearInterpolate(Alpha=NamedRerouteUsage()) {constA=1, constB=0.5}
  n60 = LinearInterpolate(Alpha=NamedRerouteUsage()) {constA=1, constB=0.5}
  n61 = NamedRerouteUsage()
  NamedRerouteDeclaration(None=Multiply(A=DotProduct(A=@n57, B=@n58), B=LinearInterpolate(A=@n59, B=@n60, Alpha=@n61))) {name=ParallaxMasking, nodeColor=<rgba 0 1 0.1094 1>}
  n62 = NamedRerouteUsage()
  n63 = Constant() {r=0.1}
  n64 = NamedRerouteUsage()
  n65 = LinearInterpolate(A=@n62, B=Multiply(A=Desaturation(None=@n62, Fraction=@n63), B=Constant() {r=24}) {constB=24}, Alpha=@n64)
  n66 = Multiply(A=Multiply(A=@n65, B=Multiply(A=@n65) {constB=0.1})) {constB=0.6}
  n67 = NamedRerouteUsage()
  n68 = NamedRerouteUsage()
  NamedRerouteDeclaration(None=LinearInterpolate(B=Multiply(A=@n66, B=@n64), Alpha=Multiply(A=@n67, B=@n68))) {name=Emissive_Out, nodeColor=<rgba 0.5703 0 1 1>}
  n69 = LinearInterpolate(
   A=VectorParameter() {group=Color, parameterName=Brake_Color, defaultValue=<rgba 0.6 0 0 1>}.RGB
   B=VectorParameter() {group=Color, parameterName=Reverse_Color, defaultValue=<rgba 0.425 0.35 0.25 1>}.RGB
   Alpha=NamedRerouteUsage()
  )
  n70 = NamedRerouteUsage()
  NamedRerouteDeclaration(None=Multiply(A=Multiply(A=@n69) {constB=2}, B=@n70)) {name=Color_Emissive, nodeColor=<rgba 0 1 0.25 1>}
  n71 = Multiply(A=NamedRerouteUsage(), B=NamedRerouteUsage())
  n72 = Clamp(None=NamedRerouteUsage()) {minDefault=0.3}
  n73 = Add(A=NamedRerouteUsage()) {constB=0.4}
  n74 = Multiply(A=Multiply(A=Multiply(A=@n71, B=@n72), B=NamedRerouteUsage()), B=Multiply(A=Saturate(None=@n73), B=NamedRerouteUsage()))
  n75 = NamedRerouteUsage()
  n76 = LinearInterpolate(A=NamedRerouteUsage(), B=NamedRerouteUsage(), Alpha=NamedRerouteUsage())
  NamedRerouteDeclaration(None=LinearInterpolate(A=Constant3Vector() {constant=<rgba 0.05 0.05 0.05 0>}, B=LinearInterpolate(A=@n74, B=@n75, Alpha=@n76)) {constAlpha=0.05}) {name=BaseColor_Out, nodeColor=<rgba 0 0.4844 1 1>}
  n77 = Add(A=@n18.Crop Mask, B=@n20.Crop Mask)
  n78 = DotProduct(A=Saturate(None=@n35.R), B=@n36)
  n79 = DotProduct(A=Saturate(None=@n38.R), B=@n39)
  n80 = Saturate(None=Ceil(None=Max(A=@n25.R, B=@n25.R)))
  n81 = Distance(A=WorldPosition().XYZ, B=CameraPositionWS())
  n82 = Clamp(None=Divide(A=Add(A=@n81) {constB=-500}) {constB=2500}) {maxDefault=0.5}
  n83 = Subtract(A=Saturate(None=AppendVector(A=@n25.R, B=@n25.R))) {constB=0.15}
  n84 = MaterialFunctionCall(Float2=Saturate(None=Multiply(A=@n83) {constB=1.1765})) {materialFunction=&BreakOutFloat2Components}
  n85 = LinearInterpolate(A=Max(A=@n84.R, B=@n84.R), B=@n80, Alpha=@n82)
  n86 = QualitySwitch(Default=LinearInterpolate(A=Max(A=@n78, B=@n79), B=@n80, Alpha=@n82), Low=@n85, Medium=@n85)
  n87 = Add(A=Saturate(None=QualitySwitch(Default=@n33.G, Low=@n25.G, Medium=@n25.G))) {constB=0.75}
  NamedRerouteDeclaration(None=Multiply(A=Multiply(A=@n77, B=@n86), B=Saturate(None=@n87))) {name=TailLightFace_Mask, nodeColor=<rgba 0 0.7891 1 1>}
  n88 = NamedRerouteDeclaration(None=@n24) {name=CropUV, nodeColor=<rgba 0 1 0.0156 1>}
  n89 = Add(A=@n88, B=Constant2Vector() {g=0.26})
  n90 = Constant2Vector() {g=2, r=3}
  n91 = Add(A=@n88, B=Constant2Vector() {g=-0.26})
  n92 = Constant2Vector() {g=-2, r=3}
  NamedRerouteDeclaration(None=LinearInterpolate(A=Multiply(A=@n89, B=@n90), B=Multiply(A=@n91, B=@n92), Alpha=NamedRerouteUsage())) {name=CroppedUVs, nodeColor=<rgba 0 1 0.0859 1>}
  NamedRerouteDeclaration(None=@n23) {name=HalfMask, nodeColor=<rgba 1 0 0.8203 1>}
  n93 = NamedRerouteUsage()
  n94 = LinearInterpolate(A=Constant2Vector() {r=-2}, B=Constant2Vector() {r=1}, Alpha=NamedRerouteUsage())
  n95 = ComponentMask(None=Multiply(A=Add(A=@n93, B=@n94) {constB=-1}, B=Constant2Vector() {g=1, r=0.5})) {r=true}
  NamedRerouteDeclaration(None=Saturate(None=Ceil(None=@n95))) {name=ReverseMask, nodeColor=<rgba 0 1 0.1094 1>}
  n96 = Constant2Vector() {r=1}
  n97 = Constant() {r=0.005}
  n98 = Constant()
  n99 = AppendVector(A=@n97, B=@n98)
  n100 = Multiply(A=NamedRerouteUsage()) {constB=0.025, desc=-0.025}
  n101 = NamedRerouteUsage()
  n102 = Multiply(A=QualitySwitch(Default=Add(A=@n100, B=@n101), Low=@n101)) {constB=50}
  n103 = Frac(None=@n102)
  n104 = Constant() {desc=.55, r=0.7}
  n105 = Constant() {r=2}
  n106 = Power(
   Base=MaterialFunctionCall!(UVs=Add(A=@n99, B=@n103), CenterPosition, Radius=@n104) {materialFunction=/Engine/Functions/Engine_MaterialFunctions01/Gradient/SphereGradient-2D}.Result 0-1
   Exp=@n105
  ) {constExponent=7}
  n107 = Saturate(None=Power(
    Base=MaterialFunctionCall(UVs=@n103, Radius=@n104) {materialFunction=&SphereGradient-2D}.Result 0-1
    Exp=@n105
   ) {constExponent=7})
  n108 = Multiply(A=@n102) {constB=0.5}
  n109 = MaterialFunctionCall(Float2=Sine(None=Add(A=@n108) {constB=0.5})) {materialFunction=&BreakOutFloat2Components}
  n110 = LinearInterpolate(A=Constant() {r=5}, B=Constant() {r=10}, Alpha=DotProduct(A=@n109.R, B=@n109.R)) {constA=5, constB=20}
  n111 = Multiply(A=Subtract(A=Saturate(None=@n106), B=@n107), B=@n110)
  n112 = Constant2Vector() {g=1}
  n113 = AppendVector(A=@n98, B=@n97)
  n114 = Power(
   Base=MaterialFunctionCall(UVs=Add(A=@n103, B=@n113), Radius=@n104) {materialFunction=&SphereGradient-2D}.Result 0-1
   Exp=@n105
  ) {constExponent=7}
  n115 = Multiply(A=@n110, B=Subtract(A=Saturate(None=@n114), B=@n107))
  NamedRerouteDeclaration(None=CrossProduct(A=AppendVector(A=@n96, B=@n111), B=AppendVector(A=@n112, B=@n115))) {name=DotsNormal, nodeColor=<rgba 0 0.6719 1 1>}
  n116 = Add(A=@n50, B=NamedRerouteUsage())
  n117 = Constant2Vector() {g=1, r=1}
  n118 = Sine(None=Frac(None=Multiply(A=@n116, B=@n117))) {period=2}
  NamedRerouteDeclaration(None=DotProduct(A=ComponentMask(None=@n118) {r=true}, B=ComponentMask(None=@n118) {g=true})) {name=ParallaxMasking_Light, nodeColor=<rgba 1 0 0.3516 1>}
  NamedRerouteDeclaration(None=ScalarParameter() {group=BP-USE-ONLY, parameterName=isBraking}) {name=isBrakingVariable, nodeColor=<rgba 0 0 0 1>}
  NamedRerouteDeclaration(None=ScalarParameter() {group=BP-USE-ONLY, parameterName=ReverseLight}) {name=ReverseLightVariable, nodeColor=<rgba 0 0 0 1>}
  NamedRerouteDeclaration(None=Multiply(A=@n9, B=NamedRerouteUsage())) {name=LightsBaseColor, nodeColor=<rgba 0 1 0.8125 1>}
  NamedRerouteDeclaration(None=Saturate(None=Multiply(B=@n57) {constA=2})) {name=ParallaxMasking_HorzLine, nodeColor=<rgba 1 0.7266 0 1>}
  n119 = Multiply(A=Constant2Vector() {g=1, r=1}, B=Add(A=NamedRerouteUsage()) {constB=0.3333})
  n120 = ComponentMask(None=Sine(None=Frac(None=@n119)) {period=2}) {r=true}
  NamedRerouteDeclaration(None=Saturate(None=ConstantBiasScale(None=@n120) {bias=0, scale=10})) {name=EncasingDivider, nodeColor=<rgba 1 0 0.4688 1>}
  n121 = Constant2Vector() {desc=-0.008, g=-0.008}
  n122 = ComponentMask(None=NamedRerouteUsage()) {g=true, r=true}
  n123 = NamedRerouteUsage()
  n124 = Add(A=@n121, B=Add(A=Multiply(A=@n122) {constB=-1}, B=@n123))
  n125 = Frac(None=@n124)
  n126 = Constant2Vector() {g=6, r=6}
  n127 = NamedRerouteUsage()
  n128 = Multiply(A=Saturate(None=ConstantBiasScale(None=@n127) {bias=-0.1}), B=NamedRerouteUsage())
  n129 = Constant() {r=0.175}
  n130 = Constant() {r=-0.175}
  n131 = NamedRerouteUsage()
  n132 = LinearInterpolate(A=@n129, B=@n130, Alpha=@n131)
  n133 = Multiply(A=ComponentMask(None=NamedRerouteUsage()) {g=true, r=true})
  n134 = Add(A=Multiply(A=@n128, B=@n132) {constB=0.175}, B=Add(A=Add(A=@n123, B=@n133), B=@n121))
  n135 = Add(A=Constant2Vector() {g=-3, r=-3}, B=Multiply(A=Frac(None=@n134)) {constB=6})
  n136 = QualitySwitch(Default=Add(A=Multiply(A=@n125, B=@n126), B=Constant2Vector() {g=-3, r=-3}), High=@n135, Epic=@n135)
  n137 = OneMinus(None=DotProduct(A=@n136, B=@n136))
  n138 = Saturate(None=ConstantBiasScale(None=Saturate(None=@n137)) {bias=-0.8, scale=5})
  n139 = NamedRerouteUsage()
  n140 = Add(A=Multiply(A=ComponentMask(None=@n139) {g=true, r=true}) {constB=0.5}, B=NamedRerouteUsage())
  n141 = Add(A=Multiply(A=Frac(None=@n140), B=Constant2Vector() {g=1.8, r=1.2}), B=Constant2Vector() {g=-0.9, r=-0.6})
  n142 = DotProduct(A=@n141, B=@n141)
  n143 = Power(Base=Saturate(None=OneMinus(None=@n142))) {constExponent=8}
  n144 = LinearInterpolate(A=@n130, B=@n129, Alpha=@n131)
  n145 = Divide(A=Add(A=Multiply(A=@n143, B=@n144) {constB=-0.175}, B=@n132), B=Constant2Vector() {g=1.2, r=1.2})
  n146 = Add(A=@n124, B=Multiply(A=Multiply(A=@n128, B=@n145) {constB=0.175}) {constB=2})
  n147 = QualitySwitch(Default=@n124, High=@n146, Epic=@n146)
  n148 = Add(A=Multiply(A=Frac(None=@n147), B=Constant2Vector() {g=3, r=3}), B=Constant2Vector() {g=-1.5, r=-1.5})
  n149 = OneMinus(None=DotProduct(A=@n148, B=@n148))
  n150 = Multiply(A=Power(Base=Saturate(None=@n149)) {constExponent=6}, B=NamedRerouteUsage())
  NamedRerouteDeclaration(None=Add(A=Multiply(A=@n138) {constB=10}, B=Multiply(A=@n150) {constB=5})) {name=ParallaxLights, nodeColor=<rgba 0 1 0.9766 1>}
  Saturate(None=Multiply(A=@n1) {constB=2.5})
  Saturate(None=Multiply(A=@n70) {constB=2.5})
 ]
}
