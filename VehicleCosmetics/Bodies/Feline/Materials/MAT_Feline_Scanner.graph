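# Surface material for the Feline vehicle body's scanner: a metallic red light strip with an animated light sweep, toggled by UseScanner.
# A sine of time, or a debug fixed time, moves the sweep through procedural cell math.
# Tiled UV dot-product masks shape lights and housing, while parallax-shifted normals and Fresnel fake lens depth.
# Quality switches simplify parallax and shaping on lower tiers; final attributes pass through MF_Vehicle_FNPostFX.
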Material /VehicleCosmetics/Bodies/Feline/Materials/MAT_Feline_Scanner
graph {
 properties {
  bAutomaticallySetUsageInEditor=false
  bUseMaterialAttributes=true
  bUsedWithMeshDeformer=true
  bUsedWithSkeletalMesh=true
  usedShadingModels=Default Lit
 }
 nodes [
  n1 = Add(A=Cosine(None=NamedRerouteUsage())) {constB=0.5}
  Ceil(None=@n1)
  Clamp(None=@n1)
  Composite() {subgraphName=Faux Lens}
  Constant2Vector() {g=-1.35, r=-0.65}
  Constant() {desc=0.175, r=0.075}
  Constant() {r=6}
  Constant() {r=0.1667}
  Constant() {desc=0.175, r=-0.075}
  LinearInterpolate()
  n2 = NamedRerouteUsage()
  n3 = Constant2Vector() {desc=-0.008, g=-0.008}
  n4 = Add(A=@n3, B=TextureCoordinate() {uTiling=8})
  n5 = Frac(None=@n4)
  n6 = Constant2Vector() {g=20, r=3}
  n7 = NamedRerouteUsage()
  n8 = Multiply(A=Saturate(None=ConstantBiasScale(None=@n7) {bias=-0.1}), B=NamedRerouteUsage())
  n9 = Add(A=Multiply(A=@n8) {constB=0.175}, B=Add(A=TextureCoordinate() {uTiling=8, vTiling=8}, B=@n3))
  n10 = Add(A=Constant2Vector() {g=-1.475, r=-1.5}, B=Multiply(A=Frac(None=@n9)) {constB=3})
  n11 = QualitySwitch(Default=Add(A=Multiply(A=@n5, B=@n6), B=Constant2Vector() {g=-1, r=-1.5}), High=@n10, Epic=@n10)
  n12 = Saturate(None=OneMinus(None=DotProduct(A=@n11, B=@n11)))
  n13 = Saturate(None=ConstantBiasScale(None=@n12) {bias=-0.8, scale=5})
  n14 = NamedRerouteUsage()
  n15 = Power(Base=@n12) {constExponent=16}
  n16 = LinearInterpolate(A=Multiply(A=Multiply(A=@n13, B=@n14)) {constB=7}, B=Multiply(A=Multiply(A=@n14, B=@n15)) {constB=150}, Alpha=NamedRerouteUsage())
  n17 = TextureCoordinate() {uTiling=8}
  n18 = Add(A=Multiply(A=Frac(None=@n17), B=Constant2Vector() {g=20, r=1.2}), B=Constant2Vector() {g=-1.35, r=-0.65})
  n19 = DotProduct(A=@n18, B=@n18)
  n20 = Power(Base=Saturate(None=OneMinus(None=@n19))) {constExponent=8}
  n21 = Divide(A=Add(A=Multiply(A=@n20) {constB=-0.175}, B=Constant() {desc=0.175, r=0.175}), B=Constant2Vector() {g=20, r=1.2})
  n22 = Add(A=@n4, B=Multiply(A=Multiply(A=@n8, B=@n21) {constB=0.175}) {constB=2})
  n23 = QualitySwitch(Default=@n4, High=@n22, Epic=@n22)
  n24 = NamedRerouteUsage()
  n25 = Fresnel() {baseReflectFraction=0, exponent=2}
  n26 = Subtract(A=Multiply(A=@n25) {constB=0.9}) {constB=0.1}
  n27 = Saturate(None=Multiply(A=Saturate(None=@n26)) {constB=2.5})
  n28 = LinearInterpolate(A=Saturate(None=ConstantBiasScale(None=@n24) {bias=-2.7, scale=-0.3}), B=@n27, Alpha=@n27)
  n29 = LinearInterpolate(A=Constant2Vector() {g=35, r=0.8}, B=Constant2Vector() {g=15, r=1.2}, Alpha=@n28)
  n30 = Constant2Vector() {g=30, r=0.8}
  n31 = Constant2Vector() {g=-2.1, r=-0.4}
  n32 = Constant2Vector() {g=-0.75, r=-0.6}
  n33 = Add(
   A=Multiply(A=Frac(None=@n23), B=QualitySwitch(Default=@n29, Low=@n30))
   B=QualitySwitch(Default=LinearInterpolate(A=@n31, B=@n32, Alpha=@n28) {constA=-0.4}, Low=Constant2Vector() {g=-1.85, r=-0.4})
  )
  n34 = DotProduct(A=@n33, B=@n33)
  n35 = Power(Base=Saturate(None=OneMinus(None=@n34))) {constExponent=6}
  n36 = NamedRerouteUsage()
  n37 = Saturate(None=Ceil(None=NamedRerouteUsage()))
  n38 = Saturate(None=NamedRerouteUsage())
  n39 = Saturate(None=Add(A=@n38) {constB=-0.75})
  n40 = Saturate(None=Add(A=Multiply(A=@n38) {constB=0.75}, B=Multiply(A=@n39) {constB=4}))
  n41 = NamedRerouteUsage()
  n42 = Multiply(
   A=Multiply(A=Multiply(A=@n35, B=@n36), B=Multiply(A=@n37, B=@n40))
   B=QualitySwitch(Default=LinearInterpolate(Alpha=@n41) {constA=1, constB=5}, Low=Constant() {r=1})
  )
  n43 = NamedRerouteDeclaration(None=Multiply(A=Add(A=@n16, B=@n42)) {constB=0.5}) {name=LightMask, nodeColor=<rgba 0.7812 1 0 1>}
  n44 = OneMinus(None=Saturate(None=@n43))
  LinearInterpolate(B=LinearInterpolate(Alpha=@n2) {constA=0.55, constB=0.65}, Alpha=@n44) {constA=0.35}
  LinearInterpolate(Alpha=@n2) {constA=0.55, constB=0.85}
  LinearInterpolate(Alpha=@n2) {constA=0.35, constB=0.55}
  n45 = SetMaterialAttributes!(
   MaterialAttributes
   Base Color=NamedRerouteUsage()
   Metallic=Constant() {r=0.9}
   Specular=Constant() {r=0.3}
   Roughness=Constant() {r=0.1}
   Emissive Color=NamedRerouteUsage()
   Normal=NamedRerouteUsage()
  ) {attributeSetTypes [
    #1
    #2
    #3
    #4
    #5
    #6
   ]}
  MaterialFunctionCall!(In=@n45) {materialFunction=/VehicleCosmetics/Materials/MaterialFunctions/MF_Vehicle_FNPostFX}
  Multiply(A=@n42) {constB=0.5}
  n46 = NamedRerouteUsage()
  n47 = Multiply(A=Power(Base=Multiply(A=@n46) {constB=0.7}) {constExponent=1.15}) {constB=5}
  NamedRerouteDeclaration(None=@n47) {name=ScannerMask_Flipbook_R, nodeColor=<rgba 0 1 0.25 1>}
  n48 = Transform(None=CameraVectorWS()) {transformSourceType=TRANSFORMSOURCE_World, transformType=TRANSFORM_Tangent}
  n49 = ComponentMask(None=@n48) {g=true, r=true}
  NamedRerouteDeclaration(None=Divide(A=Multiply(A=@n49) {constB=-1}, B=ComponentMask(None=@n48) {b=true}) {constB=1}) {name=Parallax, nodeColor=<rgba 0 1 0.2969 1>}
  n50 = NamedRerouteUsage()
  n51 = Constant3Vector() {constant=<rgba 0.8 0.8 1.8 0>}
  n52 = Transform(None=NamedRerouteUsage())
  n53 = Fresnel(Normal=LinearInterpolate(A=Multiply(A=@n51, B=@n52), B=@n52, Alpha=NamedRerouteUsage())) {baseReflectFraction=0.05, exponent=3}
  n54 = NamedRerouteUsage()
  n55 = Add(
   A=Multiply(A=Multiply(A=@n50) {constB=0.0625, desc=0.03}, B=OneMinus(None=@n53))
   B=Multiply(A=NamedRerouteUsage(), B=LinearInterpolate(Alpha=@n54) {constA=0.2})
  )
  n56 = LinearInterpolate(Alpha=Multiply(A=NamedRerouteUsage()) {constB=0.2}) {constA=0.5}
  n57 = Add(A=NamedRerouteUsage()) {constB=0.2}
  n58 = Multiply(A=@n57) {constB=1.5}
  n59 = VertexNormalWS()
  n60 = Multiply(A=Constant() {r=1.5}, B=@n59) {constB=1.5}
  n61 = CameraVectorWS()
  n62 = Saturate(None=Power(Base=DotProduct(A=@n60, B=@n61)) {constExponent=0.2})
  n63 = Multiply(A=@n57) {constB=0.85}
  n64 = DotProduct(A=@n61, B=Multiply(A=@n59, B=Constant() {r=0.45}) {constB=0.45})
  n65 = Saturate(None=Power(Base=Multiply(A=@n64) {constB=2}, Exp=Constant() {r=15}) {constExponent=50})
  n66 = ComponentMask(None=Add(A=LinearInterpolate(A=@n58, Alpha=@n62) {constB=0}, B=LinearInterpolate(B=@n63, Alpha=@n65) {constB=0.5})) {b=true, g=true, r=true}
  MaterialExpressionReroute_0 = Reroute(None=@n66)
  n67 = Add(A=Add(A=Constant() {r=0.05}, B=Multiply(A=@n55, B=@n56)), B=@MaterialExpressionReroute_0.Output)
  n68 = NamedRerouteUsage()
  n69 = NamedRerouteUsage()
  n70 = Multiply(A=Add(A=Multiply(A=@n69) {constB=0.1}, B=NamedRerouteUsage()), B=@n69)
  NamedRerouteDeclaration(None=LinearInterpolate(A=@n67, B=LinearInterpolate(A=@n67, B=@n68) {constAlpha=0.75}, Alpha=Saturate(None=@n70))) {name=BaseColor_Out, nodeColor=<rgba 0 0.4844 1 1>}
  n71 = DotProduct(A=Transform(None=Constant3Vector() {constant=<rgba 0 0 1 0>}), B=CameraVectorWS())
  n72 = OneMinus(None=Saturate(None=Multiply(B=@n71) {constA=2, desc=4}))
  NamedRerouteDeclaration(None=Saturate(None=Multiply(A=@n72, B=@n72))) {name=RimEffect, nodeColor=<rgba 0.5234 1 0 1>}
  n73 = NamedRerouteUsage()
  n74 = Multiply(A=@n73, B=@n73)
  n75 = Multiply(A=Desaturation(None=@n74, Fraction=Constant() {r=0.1}), B=Constant() {r=24}) {constB=24}
  n76 = NamedRerouteUsage()
  n77 = NamedRerouteUsage()
  n78 = Fresnel(Normal=Transform(None=@n77))
  n79 = Add(A=Multiply(A=LinearInterpolate(A=@n74, B=@n75, Alpha=@n76), B=@n76), B=Multiply(B=@n78) {constA=0.1})
  n80 = MaterialFunctionCall(Float2=@n77) {materialFunction=&BreakOutFloat2Components}
  n81 = Clamp(None=Add(A=@n80.R, B=@n80.R)) {minDefault=0.2}
  NamedRerouteDeclaration(None=Multiply(B=Multiply(A=@n79, B=@n81)) {constA=0.25}) {name=Emissive_Out, nodeColor=<rgba 0.5703 0 1 1>}
  n82 = Add(A=NamedRerouteUsage()) {constB=0.5}
  n83 = Multiply(A=Constant3Vector() {constant=<rgba 1 0.0179 0 0>}) {constB=2}
  NamedRerouteDeclaration(None=Multiply(A=Multiply(A=@n82) {constB=0.3}, B=@n83)) {name=Color_BC, nodeColor=<rgba 0 1 0.25 1>}
  n84 = NamedRerouteUsage()
  NamedRerouteDeclaration(None=Multiply(A=@n83, B=LinearInterpolate(Alpha=@n84) {constA=0.5})) {name=Color_Emissive, nodeColor=<rgba 0 1 0.25 1>}
  n85 = TextureCoordinate() {uTiling=144, vTiling=144}
  n86 = Add(A=@n85, B=Multiply(A=NamedRerouteUsage()) {constB=-0.2})
  n87 = TextureSample(UVs=QualitySwitch(Default=@n85, High=@n86, Epic=@n86)) {samplerType=SAMPLERTYPE_Normal, texture=/VehicleCosmetics/Bodies/Feline/Textures/NEO_Signage00_256_N}
  NamedRerouteDeclaration(None=@n87.RGB) {name=Normal_Detail, nodeColor=<rgba 1 0.375 0 1>}
  n88 = Sine(None=Frac(None=TextureCoordinate() {uTiling=8, vTiling=8})) {period=2}
  n89 = ComponentMask(None=@n88) {g=true}
  n90 = DotProduct(A=ComponentMask(None=@n88) {r=true}, B=@n89)
  NamedRerouteDeclaration(None=@n90) {name=ParallaxMasking_Normals, nodeColor=<rgba 0 1 0.1094 1>}
  n91 = Add(A=Constant2Vector() {g=-0.008}, B=TextureCoordinate() {uTiling=8, vTiling=8})
  n92 = NamedRerouteUsage()
  n93 = Add(A=@n91, B=Multiply(A=Multiply(A=@n92) {constB=0.125, desc=0.175}, B=NamedRerouteUsage()))
  n94 = Multiply(
   A=TextureSample(UVs=QualitySwitch(Default=@n91, High=@n93, Epic=@n93)) {samplerType=SAMPLERTYPE_Normal, texture=&NEO_Signage00_256_N}.RGB
   B=Constant3Vector() {constant=<rgba 0.75 0.75 1.2 0>}
  )
  n95 = Desaturation(None=@n94)
  NamedRerouteDeclaration(None=Add(A=@n95) {constB=0.25}) {name=Lens_Overlay, nodeColor=<rgba 0 0.3906 1 1>}
  NamedRerouteDeclaration(None=Clamp(None=@n25) {maxDefault=0.4}) {name=HousingAngle, nodeColor=<rgba 0.9453 1 0 1>}
  n96 = Multiply(A=@n95) {constB=0.2}
  n97 = Constant() {r=0.0312}
  n98 = StaticSwitchParameter(True=Add(A=NamedRerouteDeclaration(None=@n97) {name=Flipbook_B, nodeColor=<rgba 1 0.3281 0 1>}, B=@n47), False=Constant()) {defaultValue=true, group=Scanner, parameterName=UseScanner, sortPriority=-1}
  n99 = Multiply(A=Multiply(A=@n98) {constB=0.5}, B=NamedRerouteUsage()) {constB=0.25}
  n100 = Multiply(A=@n98) {constB=0.25}
  NamedRerouteDeclaration(None=Add(A=Multiply(A=@n96, B=@n98), B=Add(A=@n99, B=@n100))) {name=ScannerOverlay, nodeColor=<rgba 0 1 0.6484 1>}
  n101 = Constant3Vector() {constant=<rgba 0 0 1 0>}
  n102 = Constant3Vector() {constant=<rgba 0.75 0.75 1.2 0>}
  n103 = Add(A=@n101, B=Multiply(A=Multiply(A=@n87.RGB, B=@n102)) {constB=0.05})
  n104 = Multiply(A=@n94) {constB=0.005}
  NamedRerouteDeclaration(None=LinearInterpolate(A=@n101, B=Add(A=@n103, B=@n104), Alpha=NamedRerouteUsage())) {name=Normal_Out, nodeColor=<rgba 0.0078 0 1 1>}
  n105 = Multiply(A=@n43) {constB=0.75}
  n106 = NamedRerouteUsage()
  n107 = Multiply(A=@n106) {constB=0.3333}
  n108 = NamedRerouteUsage()
  n109 = Multiply(A=Saturate(None=@n108)) {constB=0.05}
  n110 = Add(A=@n105, B=LinearInterpolate(A=@n107, B=Add(A=@n107, B=@n109), Alpha=@n108)) {constB=0.33}
  NamedRerouteDeclaration(None=Saturate(None=Multiply(A=@n110) {constB=0.35})) {name=LightsEmissive, nodeColor=<rgba 0 1 0.8828 1>}
  n111 = ComponentMask(None=TextureCoordinate() {uTiling=8}) {r=true}
  n112 = ScalarParameter() {desc=Offsets the scanner's location., group=Scanner, parameterName=FixedTime, sortPriority=1}
  n113 = Multiply(A=Time() {period=1}, B=ScalarParameter() {defaultValue=0.85, group=Scanner, parameterName=ScannerSpeed}) {constB=0.1}
  n114 = StaticBoolParameter() {desc=Intended only for rendering team., group=Scanner, parameterName=Debug_FixedTime, sortPriority=0}
  n115 = Sine(None=NamedRerouteDeclaration(None=StaticSwitch(True=@n112, False=@n113, Value=@n114)) {name=DebugFixedTime, nodeColor=<rgba 1 0 0.2109 1>})
  n116 = Multiply(A=Multiply(A=Add(A=@n115) {constB=1.125}) {constB=0.5}) {constB=-7.1}
  n117 = Add(A=AppendVector(A=@n111, B=@n111), B=Add(A=Constant2Vector() {g=1, r=5}, B=Ceil(None=@n116))) {constB=4}
  n118 = Fmod(A=Add(A=Constant2Vector() {r=1}, B=@n117), B=Constant2Vector() {g=6, r=7})
  n119 = Truncate!(None=@n117) {is=class MaterialExpressionTruncate}
  n120 = MaterialFunctionCall(Float2=Floor(None=@n117)) {materialFunction=&BreakOutFloat2Components}
  n121 = Step(Y=@n120.R) {constX=6}
  n122 = Step(Y=@n120.G) {constX=5}
  n123 = Multiply(
   A=Multiply(A=Truncate(None=@n118), B=Constant() {r=0.1429}) {constB=0.1667}
   B=Multiply(A=Saturate(None=@n119), B=AppendVector(A=@n121, B=@n122))
  )
  n124 = Frac(None=OneMinus(None=ComponentMask(None=@n123) {g=true}))
  n125 = ComponentMask(None=Frac(None=@n123)) {r=true}
  n126 = Saturate(None=@n1)
  NamedRerouteDeclaration(None=StaticSwitchParameter(True=LinearInterpolate(A=@n124, B=@n125, Alpha=@n126), False=Constant()) {defaultValue=true, group=Scanner, parameterName=UseScanner, sortPriority=-1}) {name=FlipbookReplacement, nodeColor=<rgba 1 0.7266 0 1>}
  NamedRerouteDeclaration(None=@n27) {name=HousingAngleRefined, nodeColor=<rgba 0.6641 1 0 1>}
  n127 = Saturate(None=Add(A=@n89) {constB=-0.1})
  n128 = LinearInterpolate(Alpha=Ceil(None=NamedRerouteUsage())) {constA=1.15, constB=1.5}
  n129 = Multiply(A=NamedRerouteUsage()) {constB=3}
  NamedRerouteDeclaration(None=LinearInterpolate(A=@n90, B=Multiply(A=@n127, B=@n128) {constB=1.5}, Alpha=Saturate(None=@n129))) {name=ParallaxMasking, nodeColor=<rgba 0 1 0.1094 1>}
  NamedRerouteUsage()
  OneMinus(None=@n78)
  PinBase()
  PinBase!(Output=@n66) {reroutePins [Output {expression=&self:MaterialExpressionReroute_0}]}
  Power(Base=@n12) {constExponent=8}
  n130 = Multiply(A=@n106) {constB=0.65}
  n131 = Multiply(A=LinearInterpolate(Alpha=@n44) {constA=0.35, constB=0.35}, B=@n110) {constA=0.35}
  QualitySwitch(Default=Multiply(B=Add(A=@n105, B=@n130)) {constA=0.35}, High=@n131, Epic=@n131)
  Reroute(None=LinearInterpolate(A=NamedRerouteUsage(), Alpha=NamedRerouteUsage()))
 ]
}
