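# Physics asset for the Feline vehicle body skeletal mesh, likely a DM variant given its name suffix.
# It approximates the whole car with a single box collider on the root bone, offset forward and raised to cover the chassis.
# It has no other bodies and no constraints, so it provides simple rigid collision rather than articulated or per-part physics.
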PhysicsAsset /VehicleCosmetics/Bodies/Feline/Meshes/SK_Body_Feline_PhysicsDM
bodies {root [{shapeType=Box, extentX=444, extentY=216, extentZ=116, shapeName=root_box, center=<17.2762 0 94.5447>}]}
constraints []
